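# Niagara system for the idle effect on the FuzzyDarkness pickaxe, combining two offset glow sprites pinned to its skeletal mesh with drifting goop sprites.
# Goop_All001 spawns goop from cylinder or sphere shapes at sampled bones or sockets, inherits their motion, then drifts upward as it shrinks.
# Owner speed adjusts goop rate, lifetime and spawn group; swinging kills it, and vehicle or ADS states hide it.
# An older Goop_All variant is disabled.
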NiagaraSystem /BRCosmetics/Weapons/FORT_Melee/Pickaxe_FuzzyDarkness/FX/NS_FuzzyDarkness_Pickaxe_Idle
dependencies {
 usedDataInterfaces [&NiagaraDataInterfaceCurve, &NiagaraDataInterfaceSkeletalMesh]
 usedDynamicInputs []
 usedModules [
  &self:SystemSpawnScript.SystemScriptSource.SystemScriptGraph.NiagaraNodeAssignment_2.SetVariables_#1
  &self:SystemSpawnScript.SystemScriptSource.SystemScriptGraph.NiagaraNodeAssignment_4.SetVariables_#2
  &SystemState
  &SampleSkeletalMesh
  &self:SystemSpawnScript.SystemScriptSource.SystemScriptGraph.NiagaraNodeAssignment_3.SetVariables_#3
  &self:SystemSpawnScript.SystemScriptSource.SystemScriptGraph.NiagaraNodeAssignment_0.SetVariables_#4
 ]
}
emitters {
 Glows {
  emitterUpdateScript [
   &EmitterState(
    Inactive Response=::NewEnumerator1
    Life Cycle Mode=::NewEnumerator1
    Loop Behavior=::NewEnumerator1
    Loop Duration Mode=::NewEnumerator1
    UseLoopDelay=false
   )
   &SpawnBurst_Instantaneous(Spawn Count=2, Use Spawn Probability=false, UseLoopCountLimit=true)
  ]
  particleSpawnScript [
   &InitializeParticle(
    Color=<rgba 4 0.696 3.7797 0.1>
    Lifetime=5
    Position Offset=<12.5779 0 28.8471>
    Sprite Size Mode=::NewEnumerator3
    Uniform Sprite Size=26.8128
    UsePositionOffset=false
    Write Material Random=true
   )
   &PartitionParticles(ID to Partition:NiagaraInt32=&ReturnExecIndex())
  ]
  particleUpdateScript [
   &ParticleState(Kill Particles When Lifetime Has Elapsed=false)
   &SkeletalMeshLocation(
    Bone Sampling Mode=::NewEnumerator4
    MaskSpawnGroup=false
    Sampled Position Offset=&Add_Vector(
     A=<0 0 73>
     B=&MergeHelper_VectorSwitch(0=<7 0 0>, 1=<-7 0 0>, 2=?, 3=?, Index=$Particles.PartitionParticles.Partition, Num=::NewEnumerator0)
    )
    Skeletal Mesh=$System.Skeletal Mesh
   ) {module=SkeletalMeshLocation001}
   &CameraOffset(Camera Offset Amount=5)
  ]
  rendererClasses [&NiagaraSpriteRendererProperties]
  renderers [{rendererClass=&NiagaraSpriteRendererProperties}]
 }
 Goop_All {
  bEnabled=false
  emitterUpdateScript [
   &EmitterState(Inactive Response=::NewEnumerator1, Life Cycle Mode=::NewEnumerator1, Loop Duration=2, UseLoopDelay=false)
   /SetVariables(Emitter.MovementSpawnAlpha:NiagaraFloat=&ClampFloat(Float=&Divide_Float(A=&VectorLength(Vector=$Engine.Owner.Velocity), B=350))) {isSetParameters=true, module=SetVariables_#5}
   &SpawnRate(
    SpawnGroup:NiagaraInt32=&RandomRangeInt(Minimum=0, Maximum=2, Evaluation Type=::NewEnumerator1)
    SpawnRate=&Lerp_Float(A=8, B=6, Alpha=$Emitter.MovementSpawnAlpha)
    Use Spawn Probability=true
   )
  ]
  particleSpawnScript [
   &PartitionParticles(Number Of Partitions=4) {enabled=false}
   &InitializeParticle(
    Color=<rgba 0.8 0.8 0.8 3>
    Lifetime=&Multiply_Float(
     A=&RandomRangeFloat(Minimum=1.2, Maximum=1.6)
     B=&MakeCustomFloatFromBool(Bool=$User.bisSwinging, True Float=0.2, False Float=&Lerp_Float(A=1, B=0.4, Alpha=$Emitter.MovementSpawnAlpha))
    )
    Mesh Scale Mode=::NewEnumerator1
    Mesh Uniform Scale Max=0.4
    Mesh Uniform Scale Min=0.25
    Sprite Size Mode=::NewEnumerator1
    Uniform Sprite Size Max=5
    Uniform Sprite Size Min=10
    UsePositionOffset=false
    Write Material Random=true
   )
   &InitialMeshOrientation(Use Rotation Vector=false)
   &SkeletalMeshLocation(
    Bone / Socket Index=$Emitter.SpawnGroup
    Bone Sampling Mode=::NewEnumerator4
    MaskSpawnGroup=false
    Orientation Sampling=::NewEnumerator1
    Skeletal Mesh=$System.SkeletalMesh
   ) {enabled=false}
   &SkeletalMeshLocation(
    Bone / Socket Index=$Emitter.SpawnGroup
    Bone Sampling Mode=::NewEnumerator4
    "Inherit Velocity (Scale)"=0
    Interpolate Mesh Sampling=false
    MaskSpawnGroup=false
    Skeletal Mesh=$System.Skeletal Mesh
    Velocity Sampling=::NewEnumerator0
    VelocitySpeedLimit=false
   ) {module=SkeletalMeshLocation001}
   &ShapeLocation(
    Apply To Particle Position=::NewEnumerator1
    Cylinder Height=40
    Cylinder Height Midpoint=-0.25
    Cylinder Radius=3
    MaskSpawnGroup=false
    Override Seed=false
    Shape Primitive=::NewEnumerator1
    Surface Only=false
    UseOwnerScale=true
   )
   &ShapeLocation(
    Additional Quaternion Rotation=true
    Apply To Particle Position=::NewEnumerator1
    MaskSpawnGroup=false
    Non Uniform Scale=<0.5 1.5 1>
    Override Seed=false
    Rotation Mode=::NewEnumerator3
    Sphere Radius=15
    UseOwnerScale=true
   ) {module=ShapeLocation001}
   &AddVelocity(
    Velocity=&MergeHelper_VectorSwitch(
     0=$Particles.ShapeLocation.ShapeNormal
     1=$Particles.ShapeLocation001.ShapeNormal
     2=$Particles.ShapeLocation001.ShapeNormal
     3=$Particles.ShapeLocation001.ShapeNormal
     Index=$Emitter.SpawnGroup
     Num=::NewEnumerator2
    )
    Velocity Speed Scale=50
   ) {module=AddVelocity003}
   /SetVariables(Particles.Position=&MergeHelper_VectorSwitch(
     0=$Particles.ShapeLocation.ShapePosition
     1=$Particles.ShapeLocation001.ShapePosition
     2=$Particles.ShapeLocation001.ShapePosition
     3=$Particles.ShapeLocation001.ShapePosition
     Index=$Emitter.SpawnGroup
     Num=::NewEnumerator2
    )) {isSetParameters=true, module=SetVariables_#6}
   /SetVariables(Particles.SpawnGroup=$Emitter.SpawnGroup) {isSetParameters=true, module=SetVariables_#7}
   /SetVariables(Particles.OriginPosition:NiagaraPosition=&TransformPosition(Position=$Particles.Position, Source Space=::World, Destination Space=::Local)) {isSetParameters=true, module=SetVariables_#8}
   /SetVariables(Particles.TargetPos:NiagaraPosition=$Particles.Position) {isSetParameters=true, module=SetVariables_#9}
   /SetVariables() {isSetParameters=true, module=SetVariables_#10}
  ]
  particleUpdateScript [
   &ParticleState()
   &SkeletalMeshLocation(
    Bone / Socket Index=$Particles.SpawnGroup
    Bone Sampling Mode=::NewEnumerator4
    MaskSpawnGroup=false
    Position Sampling=::NewEnumerator2
    Skeletal Mesh=$System.Skeletal Mesh
   ) {module=SkeletalMeshLocation002}
   /SetVariables(Particles.OriginPosition:NiagaraPosition=&TransformPosition(Position=$Particles.Initial.OriginPosition, Source Space=::Local, Destination Space=::World)) {isSetParameters=true, module=SetVariables_#11}
   /SetVariables(Particles.TargetPos:NiagaraPosition=/Niagara/DynamicInputs/Lerp/SmoothLerpOverTime_Position(Convergence Rate:NiagaraFloat=3.5, Smooth Value:NiagaraPosition=$Particles.TargetPos, Target Value:NiagaraPosition=$Particles.OriginPosition)) {isSetParameters=true, module=SetVariables_#12}
   /SetVariables(Particles.DistanceToOrigin:NiagaraFloat=&DistanceBetweenPositions(A=Linked() {linkedVariable {name=System.PickaxeMainBody, type=Vector3f}}, B=$Particles.Position)) {isSetParameters=true, module=SetVariables_#13}
   &ScaleSpriteSize(
    Scale Sprite Size Mode=::NewEnumerator0
    Uniform Scale Factor=&FloatFromCurve(FloatCurve=DataInterface() {propertyValues {Curve {keys [
         {arriveTangent=9.1187, interpMode=RCIM_Cubic, time=0, value=0}
         {arriveTangent=0.5833, interpMode=RCIM_Cubic, leaveTangent=0.5833, time=0.2487, value=1.1246}
         {arriveTangent=-1.497, interpMode=RCIM_Cubic, leaveTangent=-1.497, time=0.6, value=0.35}
         <at 1 0 RCIM_Cubic>
        ]}}})
   )
   &AccelerationForce(Acceleration=&Multiply_VectorByFloat(
     Vector=&RandomRangeVector(Minimum=<-3 -3 5>, Maximum=<3 3 10>, Evaluation Type=::NewEnumerator1)
     Float=1
    ))
   &Drag(Drag=5, Use Linear Drag=true, Use Rotational Drag=true)
   &GravityForce(Gravity=&MakeVector(
     X=0
     Y=0
     Z=&FloatFromCurve(FloatCurve=DataInterface() {propertyValues {Curve {keys [<at 0.3 0>,<at 1 1>]}}}, Scale Curve=400)
    ))
   &DynamicMaterialParameters(
    Index 0 Param 1=5
    Index 0 Param 2=12
    Index 0 Param 3=&FloatFromCurve(FloatCurve=DataInterface() {propertyValues {Curve {keys [<at 0 5 RCIM_Cubic>, {arriveTangent=-4.1151, interpMode=RCIM_Cubic, leaveTangent=-4.1151, tangentMode=RCTM_User, time=1, value=1}]}}})
    Module.Param0WriteEnabled=true
    Module.Param1WriteEnabled=true
    Module.Param2WriteEnabled=true
    Module.Param3WriteEnabled=false
   )
   &InheritSourceMovement(
    Applied Force Scale=&VectorFromFloat(Value=&FloatFromCurve(FloatCurve=DataInterface() {propertyValues {Curve {keys [<at 0 0>,<at 1 -0.5>]}}}))
    Applied Position Scale=&VectorFromFloat(Value=&MakeCustomFloatFromBool(
      Bool=$User.IsSkydiving
      True Float=&FloatFromCurve(FloatCurve=DataInterface() {propertyValues {Curve {keys [<at 0 0.95 RCIM_Cubic>,<at 1 0.8 RCIM_Cubic>]}}})
      False Float=&FloatFromCurve(FloatCurve=DataInterface() {propertyValues {Curve {keys [<at 0 0.95 RCIM_Cubic>,<at 1 0.8 RCIM_Cubic>]}}})
     ))
    ApplyForce=true
    ApplyForceFalloff=false
    ApplyPosition=true
    ApplyPositionFalloff=false
    Module.LimitForceSpeed=false
   ) {enabled=false, module=InheritSourceMovement001}
   &InheritSourceMovement(
    Applied Force Scale=&VectorFromFloat(Value=-1)
    Applied Position Scale=&VectorFromFloat(Value=&FloatFromCurve(FloatCurve=DataInterface() {propertyValues {Curve {keys [{arriveTangent=-2.0252, interpMode=RCIM_Cubic, leaveTangent=-2.0252, tangentMode=RCTM_User, time=0, value=1}, <at 1 0 RCIM_Cubic>]}}}))
    ApplyForce=true
    ApplyForceFalloff=false
    ApplyPosition=true
    ApplyPositionFalloff=false
    Module.LimitForceSpeed=false
    Source Velocity:Vector3f=$Particles.SkeletalMeshLocation002.SampledVelocity
   )
   &PointForce(Force Origin=$System.PickaxeMainBody, Module.Use Falloff=false, Module.UseLinearForce=false) {enabled=false}
   /Niagara/Modules/Update/Forces/SpringForce(
    " Particle Equilibrium Position":NiagaraPosition=$Particles.OriginPosition
    Dampening Coefficient=3
    Force Strength=1.5
    Spring Tightness=100
   ) {enabled=false}
   /SetVariables(Particles.Position=&Lerp_Vector(
     A=Linked() {linkedVariable {name=Particles.Position, type=NiagaraPosition}}
     B=Linked() {linkedVariable {name=Particles.TargetPos, type=NiagaraPosition}}
     Alpha=$Emitter.MovementSpawnAlpha
    )) {enabled=false, isSetParameters=true, module=SetVariables_#14}
   &LinearForce(
    Coordinate Space:ENiagaraCoordinateSpace=::World
    Force=&Multiply_VectorByFloat(
     Vector=&NormalizeVector(
      Vector To Normalize=&SubtractVectorFromPosition(Position=$Particles.OriginPosition, Vector=Linked() {linkedVariable {name=Particles.Position, type=NiagaraPosition}})
      Safe Normalize=true
      Safe Fallback Threshold=0
      Safe Fallback Vector=<0 0 0.001>
     )
     Float=&RemapRange(
      Input Value To Remap=$Emitter.MovementSpawnAlpha
      Input Min=0
      Input Max=1
      Output Min=0
      Output Max=&RandomRangeFloat(Minimum=3500, Maximum=7500)
     )
    )
   ) {enabled=false}
   &PointAttractionForce(
    Attraction Radius=250
    Attraction Strength=&Multiply_Float(A=500, B=$Emitter.MovementSpawnAlpha)
    Attractor Position:Vector3f=$System.PickaxeMainBody
    Kill Within Radius=false
    Module.Use Falloff=true
    UseAttractorOffset=false
   )
   &SolveForcesAndVelocity(Clamp Velocity=false, Limit Acceleration=false)
   &ScaleColor(
    Scale Alpha=&MakeCustomFloatFromBool(Bool=&Boolean_OrOperation(Bool1=$User.bisInVehicle, Bool2=$Emitter.ADS), True Float=0, False Float=1)
    ScaleA=true
    ScaleRGB=true
   ) {module=ScaleColor001}
   &ScaleColor(
    Scale RGB=&VectorFromFloat(Value=&FNTimeOfDayScalar(Night Scale=0.6, Morning Scale=1, Evening Scale=1))
    ScaleA=true
    ScaleRGB=true
   ) {module=ScaleColor002}
   &CameraOffset(Camera Offset Amount=&Multiply_Float(
     A=&FloatFromCurve(FloatCurve=DataInterface() {propertyValues {Curve {keys [{interpMode=RCIM_Cubic, tangentMode=RCTM_User, time=0.5, value=1}, {interpMode=RCIM_Cubic, tangentMode=RCTM_User, time=1, value=0}]}}})
     B=&MergeHelper_FloatSwitch(0=2, 1=5, 2=0, 3=0, 4=0, Index=$Particles.SpawnGroup, Num=::NewEnumerator3)
    ))
   &InterpolateOverTime() {enabled=false}
   /SetVariables(Particles.TargetPos:NiagaraPosition=&SmoothLerpOverTime_Position(
     Convergence Rate:NiagaraFloat=5
     Smooth Value:NiagaraPosition=$Particles.TargetPos
     Target Value:NiagaraPosition=&LerpPosition(
      A=$Particles.Position
      B=$Particles.OriginPosition
      Alpha=/Niagara/DynamicInputs/Lerp/SmoothLerpOverTime_Float(
       Convergence Rate:NiagaraFloat=55
       Smooth Value:NiagaraFloat=$Particles.BlendPositions
       Target Value:NiagaraFloat=&MakeCustomFloatFromBool(Bool=$User.bisSwinging, True Float=1, False Float=0)
      )
     )
    )) {enabled=false, isSetParameters=true, module=SetVariables_#15}
   &RegionMask(Inner Float=0, Outer Float=1, Region Falloff=0.1, Region Origin:Vector3f=$System.PickaxeMainBody, UseOffset=false, UseRotation=false)
   &ScaleSpriteSize(
    Uniform Curve Index=$Particles.RegionMask.Float
    Uniform Curve Sprite Scale=DataInterface() {propertyValues {Curve {keys [<at 0 1 RCIM_Cubic>,<at 1 0 RCIM_Cubic>]}}}
   ) {module=ScaleSpriteSize001}
   &KillParticles(Kill Particles=&SetBoolByFloatComparison(A=$Particles.RegionMask.Float, B=0.99))
  ]
  rendererClasses [&NiagaraSpriteRendererProperties]
  renderers [
   {rendererClass=&NiagaraSpriteRendererProperties}
   {rendererClass=&NiagaraSpriteRendererProperties, rendererIndex=1}
  ]
 }
 Goop_All001 {
  emitterUpdateScript [
   &EmitterState(Inactive Response=::NewEnumerator1, Life Cycle Mode=::NewEnumerator1, Loop Duration=2, UseLoopDelay=false)
   /SetVariables(Emitter.MovementSpawnAlpha:NiagaraFloat=&ClampFloat(Float=&Divide_Float(A=&VectorLength(Vector=$Engine.Owner.Velocity), B=350))) {isSetParameters=true, module=SetVariables_#5}
   &SpawnRate(
    SpawnGroup:NiagaraInt32=&RandomRangeInt(
     Minimum=&MakeCustomIntFromBool(Bool=&SetBoolByFloatComparison(A=$Emitter.MovementSpawnAlpha, B=0.5), True INT=1, False INT=0)
     Maximum=2
     Evaluation Type=::NewEnumerator1
    )
    SpawnRate=&Lerp_Float(A=8, B=9, Alpha=$Emitter.MovementSpawnAlpha)
    Use Spawn Probability=true
   )
  ]
  particleSpawnScript [
   &PartitionParticles(Number Of Partitions=4) {enabled=false}
   &InitializeParticle(
    Color=<rgba 0.8 0.8 0.8 3>
    Lifetime=&Multiply_Float(
     A=&RandomRangeFloat(Minimum=1.2, Maximum=1.6)
     B=&MakeCustomFloatFromBool(Bool=$User.bisSwinging, True Float=0.2, False Float=&Lerp_Float(A=1, B=0.5, Alpha=$Emitter.MovementSpawnAlpha))
    )
    Mesh Scale Mode=::NewEnumerator1
    Mesh Uniform Scale Max=0.4
    Mesh Uniform Scale Min=0.25
    Sprite Size Mode=::NewEnumerator1
    Uniform Sprite Size Max=5
    Uniform Sprite Size Min=10
    UsePositionOffset=false
    Write Material Random=true
   )
   &InitialMeshOrientation(Use Rotation Vector=false)
   &SkeletalMeshLocation(
    Bone / Socket Index=$Emitter.SpawnGroup
    Bone Sampling Mode=::NewEnumerator4
    MaskSpawnGroup=false
    Orientation Sampling=::NewEnumerator1
    Skeletal Mesh=$System.SkeletalMesh
   ) {enabled=false}
   &SkeletalMeshLocation(
    Bone / Socket Index=$Emitter.SpawnGroup
    Bone Sampling Mode=::NewEnumerator4
    "Inherit Velocity (Scale)"=0
    Interpolate Mesh Sampling=false
    MaskSpawnGroup=false
    Skeletal Mesh=$System.Skeletal Mesh
    Velocity Sampling=::NewEnumerator0
    VelocitySpeedLimit=false
   ) {module=SkeletalMeshLocation001}
   &ShapeLocation(
    Apply To Particle Position=::NewEnumerator1
    Cylinder Height=40
    Cylinder Height Midpoint=-0.25
    Cylinder Radius=3
    MaskSpawnGroup=false
    Override Seed=false
    Shape Primitive=::NewEnumerator1
    Surface Only=false
    UseOwnerScale=true
   )
   &ShapeLocation(
    Additional Quaternion Rotation=true
    Apply To Particle Position=::NewEnumerator1
    MaskSpawnGroup=false
    Non Uniform Scale=<0.5 1.5 1>
    Override Seed=false
    Rotation Mode=::NewEnumerator3
    Sphere Radius=15
    UseOwnerScale=true
   ) {module=ShapeLocation001}
   &AddVelocity(
    Velocity=&MergeHelper_VectorSwitch(
     0=$Particles.ShapeLocation.ShapeNormal
     1=$Particles.ShapeLocation001.ShapeNormal
     2=$Particles.ShapeLocation001.ShapeNormal
     3=$Particles.ShapeLocation001.ShapeNormal
     Index=$Emitter.SpawnGroup
     Num=::NewEnumerator2
    )
    Velocity Speed Scale=&MakeCustomFloatFromBool(Bool=$User.bIsFrontEndPreview, True Float=20, False Float=50)
   ) {module=AddVelocity003}
   /SetVariables(Particles.Position=&MergeHelper_VectorSwitch(
     0=$Particles.ShapeLocation.ShapePosition
     1=$Particles.ShapeLocation001.ShapePosition
     2=$Particles.ShapeLocation001.ShapePosition
     3=$Particles.ShapeLocation001.ShapePosition
     Index=$Emitter.SpawnGroup
     Num=::NewEnumerator2
    )) {isSetParameters=true, module=SetVariables_#6}
   /SetVariables(Particles.SpawnGroup=$Emitter.SpawnGroup) {isSetParameters=true, module=SetVariables_#7}
   /SetVariables(Particles.OriginPosition:NiagaraPosition=&TransformPosition(Position=$Particles.Position, Source Space=::World, Destination Space=::Local)) {isSetParameters=true, module=SetVariables_#8}
   /SetVariables(Particles.TargetPos:NiagaraPosition=$Particles.Position) {isSetParameters=true, module=SetVariables_#9}
   /SetVariables() {isSetParameters=true, module=SetVariables_#10}
  ]
  particleUpdateScript [
   &ParticleState()
   &SkeletalMeshLocation(
    Bone / Socket Index=$Particles.SpawnGroup
    Bone Sampling Mode=::NewEnumerator4
    MaskSpawnGroup=false
    Position Sampling=::NewEnumerator2
    Skeletal Mesh=$System.Skeletal Mesh
   ) {module=SkeletalMeshLocation002}
   /SetVariables(Particles.OriginPosition:NiagaraPosition=&TransformPosition(Position=$Particles.Initial.OriginPosition, Source Space=::Local, Destination Space=::World)) {isSetParameters=true, module=SetVariables_#11}
   /SetVariables(Particles.TargetPos:NiagaraPosition=&SmoothLerpOverTime_Position(
     Convergence Rate:NiagaraFloat=&RandomRangeFloat(Minimum=4, Maximum=32)
     Smooth Value:NiagaraPosition=$Particles.TargetPos
     Target Value:NiagaraPosition=$Particles.OriginPosition
    )) {isSetParameters=true, module=SetVariables_#12}
   /SetVariables(Particles.DistanceToOrigin:NiagaraFloat=&DistanceBetweenPositions(A=Linked() {linkedVariable {name=System.PickaxeMainBody, type=Vector3f}}, B=$Particles.Position)) {isSetParameters=true, module=SetVariables_#13}
   &ScaleSpriteSize(
    Scale Sprite Size Mode=::NewEnumerator0
    Uniform Scale Factor=&FloatFromCurve(FloatCurve=DataInterface() {propertyValues {Curve {keys [
         {arriveTangent=9.1187, interpMode=RCIM_Cubic, time=0, value=0}
         {arriveTangent=0.5833, interpMode=RCIM_Cubic, leaveTangent=0.5833, time=0.2487, value=1.1246}
         {arriveTangent=-1.497, interpMode=RCIM_Cubic, leaveTangent=-1.497, time=0.6, value=0.35}
         <at 1 0 RCIM_Cubic>
        ]}}})
   )
   &AccelerationForce(Acceleration=&Multiply_VectorByFloat(
     Vector=&RandomRangeVector(Minimum=<-3 -3 5>, Maximum=<3 3 10>, Evaluation Type=::NewEnumerator1)
     Float=1
    ))
   &Drag(Drag=&RandomRangeFloat(Minimum=1, Maximum=6), Use Linear Drag=true, Use Rotational Drag=true)
   &GravityForce(Gravity=&MakeVector(
     X=0
     Y=0
     Z=&FloatFromCurve(FloatCurve=DataInterface() {propertyValues {Curve {keys [<at 0.3 0>,<at 1 1>]}}}, Scale Curve=400)
    ))
   &DynamicMaterialParameters(
    Index 0 Param 1=5
    Index 0 Param 2=12
    Index 0 Param 3=&FloatFromCurve(FloatCurve=DataInterface() {propertyValues {Curve {keys [<at 0 5 RCIM_Cubic>, {arriveTangent=-4.1151, interpMode=RCIM_Cubic, leaveTangent=-4.1151, tangentMode=RCTM_User, time=1, value=1}]}}})
    Module.Param0WriteEnabled=true
    Module.Param1WriteEnabled=true
    Module.Param2WriteEnabled=true
    Module.Param3WriteEnabled=false
   )
   &InheritSourceMovement(
    Applied Force Scale=&VectorFromFloat(Value=&FloatFromCurve(FloatCurve=DataInterface() {propertyValues {Curve {keys [<at 0 0>,<at 1 -0.5>]}}}))
    Applied Position Scale=&VectorFromFloat(Value=&MakeCustomFloatFromBool(
      Bool=$User.IsSkydiving
      True Float=&FloatFromCurve(FloatCurve=DataInterface() {propertyValues {Curve {keys [<at 0 0.95 RCIM_Cubic>,<at 1 0.8 RCIM_Cubic>]}}})
      False Float=&FloatFromCurve(FloatCurve=DataInterface() {propertyValues {Curve {keys [<at 0 0.95 RCIM_Cubic>,<at 1 0.8 RCIM_Cubic>]}}})
     ))
    ApplyForce=true
    ApplyForceFalloff=false
    ApplyPosition=true
    ApplyPositionFalloff=false
    Module.LimitForceSpeed=false
   ) {enabled=false, module=InheritSourceMovement001}
   &InheritSourceMovement(
    Applied Force Scale=<0 0 0>
    Applied Position Scale=&VectorFromFloat(Value=&FloatFromCurve(
      FloatCurve=DataInterface() {propertyValues {Curve {keys [<at 0 1 RCIM_Cubic>,<at 1 0 RCIM_Cubic>]}}}
      Scale Curve=&Lerp_Float(A=1, B=0.85, Alpha=$Emitter.MovementSpawnAlpha)
     ))
    ApplyForce=true
    ApplyForceFalloff=false
    ApplyPosition=true
    ApplyPositionFalloff=false
    Module.LimitForceSpeed=false
    Source Velocity:Vector3f=$Particles.SkeletalMeshLocation002.SampledVelocity
   )
   &PointForce(Force Origin=$System.PickaxeMainBody, Module.Use Falloff=false, Module.UseLinearForce=false) {enabled=false}
   &SpringForce(
    " Particle Equilibrium Position":NiagaraPosition=$Particles.OriginPosition
    Dampening Coefficient=3
    Force Strength=1.5
    Spring Tightness=100
   ) {enabled=false}
   /SetVariables(Particles.Position=&Lerp_Vector(
     A=Linked() {linkedVariable {name=Particles.Position, type=NiagaraPosition}}
     B=Linked() {linkedVariable {name=Particles.TargetPos, type=NiagaraPosition}}
     Alpha=$Emitter.MovementSpawnAlpha
    )) {enabled=false, isSetParameters=true, module=SetVariables_#14}
   &LinearForce(
    Coordinate Space:ENiagaraCoordinateSpace=::World
    Force=&Multiply_VectorByFloat(
     Vector=&NormalizeVector(
      Vector To Normalize=&SubtractVectorFromPosition(Position=$Particles.OriginPosition, Vector=Linked() {linkedVariable {name=Particles.Position, type=NiagaraPosition}})
      Safe Normalize=true
      Safe Fallback Threshold=0
      Safe Fallback Vector=<0 0 0.001>
     )
     Float=&RemapRange(
      Input Value To Remap=$Emitter.MovementSpawnAlpha
      Input Min=0
      Input Max=1
      Output Min=0
      Output Max=&RandomRangeFloat(Minimum=3500, Maximum=7500)
     )
    )
   ) {enabled=false}
   &SolveForcesAndVelocity(Clamp Velocity=false, Limit Acceleration=false)
   &ScaleColor(
    Scale Alpha=&MakeCustomFloatFromBool(Bool=&Boolean_OrOperation(Bool1=$User.bisInVehicle, Bool2=$Emitter.ADS), True Float=0, False Float=1)
    ScaleA=true
    ScaleRGB=true
   ) {module=ScaleColor001}
   &ScaleColor(
    Scale RGB=&VectorFromFloat(Value=&FNTimeOfDayScalar(Night Scale=0.6, Morning Scale=1, Evening Scale=1))
    ScaleA=true
    ScaleRGB=true
   ) {module=ScaleColor002}
   &CameraOffset(Camera Offset Amount=&Multiply_Float(
     A=&FloatFromCurve(FloatCurve=DataInterface() {propertyValues {Curve {keys [{interpMode=RCIM_Cubic, tangentMode=RCTM_User, time=0.5, value=1}, {interpMode=RCIM_Cubic, tangentMode=RCTM_User, time=1, value=0}]}}})
     B=&MergeHelper_FloatSwitch(0=0, 1=0, 2=0, 3=0, 4=0, Index=$Particles.SpawnGroup, Num=::NewEnumerator3)
    )) {enabled=false}
   /SetVariables(Particles.TargetPos:NiagaraPosition=&SmoothLerpOverTime_Position(
     Convergence Rate:NiagaraFloat=5
     Smooth Value:NiagaraPosition=$Particles.TargetPos
     Target Value:NiagaraPosition=&LerpPosition(
      A=$Particles.Position
      B=$Particles.OriginPosition
      Alpha=&SmoothLerpOverTime_Float(
       Convergence Rate:NiagaraFloat=55
       Smooth Value:NiagaraFloat=$Particles.BlendPositions
       Target Value:NiagaraFloat=&MakeCustomFloatFromBool(Bool=$User.bisSwinging, True Float=1, False Float=0)
      )
     )
    )) {enabled=false, isSetParameters=true, module=SetVariables_#15}
   &RegionMask(Inner Float=0, Outer Float=1, Region Falloff=0.1, Region Origin:Vector3f=$System.PickaxeMainBody, UseOffset=false, UseRotation=false) {enabled=false}
   &ScaleSpriteSize(
    Uniform Curve Index=$Particles.RegionMask.Float
    Uniform Curve Sprite Scale=DataInterface() {propertyValues {Curve {keys [<at 0 1 RCIM_Cubic>,<at 1 0 RCIM_Cubic>]}}}
   ) {enabled=false, module=ScaleSpriteSize001}
   &KillParticles(Kill Particles=$User.bisSwinging)
  ]
  rendererClasses [&NiagaraSpriteRendererProperties]
  renderers [
   {rendererClass=&NiagaraSpriteRendererProperties}
   {rendererClass=&NiagaraSpriteRendererProperties, rendererIndex=1}
  ]
 }
}
summary {
 systemName=NS_FuzzyDarkness_Pickaxe_Idle
 emitters [
  {emitterName=Glows, rendererClasses [&NiagaraSpriteRendererProperties]}
  {bEnabled=false, emitterName=Goop_All, rendererClasses [&NiagaraSpriteRendererProperties]}
  {emitterName=Goop_All001, rendererClasses [&NiagaraSpriteRendererProperties]}
 ]
 userVariables [
  User.bIsFrontEnd {type=NiagaraBool}
  User.bIsFrontEndPreview {type=NiagaraBool}
  User.bisHandHidden {type=NiagaraBool}
  User.bisInVehicle {type=NiagaraBool}
  User.bisSwinging {type=NiagaraBool}
  User.IsSkydiving {type=NiagaraBool}
  User.RibbonIDOffset {type=NiagaraInt32}
 ]
}
